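# Material function that builds an animated square glitch pattern, outputting pixelated UVs, two square masks and a time value offset by custom primitive data.
# Time-seeded hashes on two floored UV grids, one rotatable, choose between two pixelation resolutions and multiply into the masks.
# A separate frequency mask combines a sine pulse, desynchronized by per-instance random, actor position and primitive data, with random flicker, computed per vertex through a VertexInterpolator.
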MaterialFunction /SoberBotany/Materials/Functions/MF_DarkNexus_SquarePattern
graph {nodes [
  n1 = FunctionInput(Preview=TextureCoordinate()) {bUsePreviewValueAsDefault=true, inputName=UV, inputType=FunctionInput_Vector2, sortPriority=4}
  n2 = MaterialFunctionCall(UVs=@n1, "Rotation Angle (0-1)"=ScalarParameter() {group=HolographicTexture_Glitch, parameterName=GlitchUVRotation}) {materialFunction=&CustomRotator}
  n3 = Divide(A=NamedRerouteUsage()) {constB=40}
  n4 = Multiply(
   A=@n2.Rotated Values
   B=FunctionInput(Preview=ScalarParameter() {defaultValue=5, group=HolographicTexture_Glitch, parameterName=GlitchMaskResolution}) {bUsePreviewValueAsDefault=true, inputName=GlitchMaskResolution, inputType=FunctionInput_Scalar, sortPriority=2}
  )
  n5 = MaterialFunctionCall!(Seed=@n3, Time=@n3, Vec3, Vec2=Floor(None=Add(A=@n4) {constB=0.37})) {materialFunction=&/MF_CustomHash}
  n6 = LinearInterpolate(
   A=FunctionInput(Preview=ScalarParameter() {defaultValue=12, group=HolographicTexture_Glitch, parameterName=PixellateGlitch_MaxResolution}) {bUsePreviewValueAsDefault=true, inputName=PixellateGlitch_MaxResolution, inputType=FunctionInput_Scalar, sortPriority=2}
   B=FunctionInput(Preview=ScalarParameter() {defaultValue=4, group=HolographicTexture_Glitch, parameterName=PixellateGlitch_MinResolution}) {bUsePreviewValueAsDefault=true, inputName=PixellateGlitch_MinResolution, inputType=FunctionInput_Scalar, sortPriority=2}
   Alpha=Round(None=Clamp(None=@n5.Float))
  )
  n7 = Multiply(A=@n2.Rotated Values, B=@n6)
  n8 = ConstantBiasScale(None=@n5.Vec2)
  n9 = ComponentMask(None=@n8) {r=true}
  n10 = LinearInterpolate(A=@n1, B=Divide(A=Floor(None=@n7), B=@n6), Alpha=Round(None=@n9))
  ComponentMask(None=FunctionOutput(None=LinearInterpolate(A=@n10, B=@n1)) {outputName=Glitched UV, sortPriority=2}) {r=true}
  Constant() {r=-0.8415}
  FunctionInput(Preview=Constant() {r=0.075}) {bUsePreviewValueAsDefault=true, inputName=GlitchDuration, inputType=FunctionInput_Scalar, sortPriority=1}
  n11 = Multiply(A=@n1) {constB=3.7}
  n12 = MaterialFunctionCall(Seed=@n3, Time=Add(A=@n3) {constB=0.5}, Vec2=Floor(None=Add(A=@n11) {constB=0.67})) {materialFunction=&MF_CustomHash}
  n13 = ConstantBiasScale(None=@n12.Vec2)
  FunctionOutput(None=Multiply(A=@n9, B=ComponentMask(None=@n13) {g=true})) {outputName=Square Mask}
  FunctionOutput() {outputName=GlitchFrequencyMask}
  FunctionOutput(None=Multiply(A=ComponentMask(None=@n8) {g=true}, B=ComponentMask(None=@n13) {r=true})) {outputName=Square Mask 2}
  n14 = ScalarParameter() {bUseCustomPrimitiveData=true, parameterName=Offset Time, primitiveDataIndex=2}
  n15 = FunctionInput(Preview=Time()) {bUsePreviewValueAsDefault=true, inputName=Time, inputType=FunctionInput_Scalar}
  FunctionOutput(None=NamedRerouteDeclaration(None=Add(A=@n14, B=@n15)) {name=Random time, nodeColor=<rgba 0 1 0.7422 1>}) {outputName=Random time}
  LinearInterpolate(Alpha=NamedRerouteUsage()) {constA=50, constB=49.89}
  LinearInterpolate(A=@n10, B=@n1, Alpha=Round(None=ConstantBiasScale(None=@n12.Float)))
  n16 = Constant() {r=1}
  n17 = FunctionInput(Preview=Constant() {r=0.1}) {bUsePreviewValueAsDefault=true, inputName=GlitchFrequency, inputType=FunctionInput_Scalar, sortPriority=1}
  n18 = Multiply(A=@n17) {constB=0.5}
  n19 = PerInstanceRandom()
  n20 = ScalarParameter() {bUseCustomPrimitiveData=true, parameterName=Offset Time, primitiveDataIndex=2}
  n21 = MaterialFunctionCall(Float3=Divide(A=ActorPositionWS()) {constB=2000}) {materialFunction=&BreakOutFloat3Components}
  n22 = Add(A=@n21.R, B=OneMinus(None=@n21.G))
  n23 = Add(A=Add(A=LinearInterpolate(Alpha=@n19) {constB=12335.3467}, B=Add(A=@n20, B=@n22)), B=Time())
  n24 = Multiply(A=Sine(None=Multiply(A=@n23, B=@n17) {constB=0.2})) {constB=0.5}
  n25 = Constant()
  n26 = Saturate(None=Add(
    A=MaterialFunctionCall() {materialFunction=/Game/MF__FloatFlickerRandom_Generic}.RandomFlickerFloat
    B=FunctionInput(Preview=@n25) {bUsePreviewValueAsDefault=true, inputName=Glitch Flickering, inputType=FunctionInput_Scalar, sortPriority=1}
   ))
  n27 = Multiply(
   A=SmoothStep(Min=Subtract(A=@n16, B=@n18), Max=@n16, Value=Add(A=@n24) {constB=0.5}) {constMax=0.9, constMin=0.8}
   B=Multiply(A=@n26, B=@n26)
  )
  NamedRerouteDeclaration(None=VertexInterpolator(VS=@n27).PS) {name=GlitchFrequencyMask, nodeColor=<rgba 0.9377 1 0.8953 1>}
  OneMinus(None=@n27)
  ScalarParameter() {defaultValue=0.1, group=HolographicTexture_Glitch, parameterName=GlitchFrequency, sortPriority=10}
  ScalarParameter() {defaultValue=0.075, group=HolographicTexture_Glitch, parameterName=GlitchDuration, sortPriority=10}
 ]}
