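# Material function that outputs an animated, glowing hexagon pattern as emissive color, grouped under DarkNexes hex parameters.
# Object-relative world position feeds a dithered, wavy-distorted triplanar function supplying UVs and derivatives to sample a hexagon mask.
# The green channel forms sharpness-controlled lines, while a time-offset sine of red adds pulsing tile shading, blending base and highlight colors.
# A ray tracing quality switch applies a separate Lumen emissive multiplier.
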MaterialFunction /SoberBotany/Materials/MF_HexagonPattern
graph {nodes [
  n1 = ScalarParameter() {defaultValue=0.05, group=DarkNexes_Hexes, parameterName=Hexes_Sharpness}
  n2 = ScalarParameter() {defaultValue=1, group=DarkNexes_Hexes, parameterName=Hexes_Tiling}
  n3 = WorldPosition() {worldPositionShaderOffset=WPT_CameraRelative}
  n4 = MaterialFunctionCall!(
   Absolute World Position=Subtract(A=TruncateLWC!(None=@n3.XYZ), B=ObjectPositionWS() {originType=CameraRelative})
   "Normal (WS)"
   Texture Scale
   Triplanar Fade
   Derivatives Scale
  ) {materialFunction=/SoberBotany/Materials/Functions/MF_TriplanarDithered_WavyDistortion}
  n5 = TextureSample!(
   UVs=Multiply(A=Constant2Vector() {g=1, r=0.85}, B=Multiply(A=@n2, B=@n4.UVs) {constB=0.2})
   Tex
   "DDX(UVs)"=@n4.DDX
   "DDY(UVs)"=@n4.DDX
   Apply View MipBias
  ) {
   mipValueMode=TMVM_Derivative
   samplerSource=SSM_Wrap_WorldGroupSettings
   texture=/Game/Global/Materials/MaterialGym/Materials/HexDisplacement/Textures/T_Hexagon_Displacement_Mask
  }
  n6 = ComponentMask(None=@n5.RGB) {g=true}
  n7 = Sine(None=Subtract(A=ComponentMask(None=@n5.RGB) {r=true}, B=NamedRerouteUsage()))
  n8 = Add(
   A=OneMinus(None=SmoothStep(Max=@n1, Value=@n6) {constMax=0.3})
   B=Multiply(A=OneMinus(None=@n7), B=ScalarParameter() {defaultValue=0.1, group=DarkNexes_Hexes, parameterName=HexesTileDepthColor}) {constB=0.1}
  )
  n9 = LinearInterpolate(
   A=VectorParameter() {group=DarkNexes_Hexes, parameterName=HexesColor, defaultValue=<rgba 0.7344 0.1916 0 1>}.RGB
   B=VectorParameter() {group=DarkNexes_Hexes, parameterName=HexesHighlightColor, defaultValue=<rgba 14 4.3826 0 1>}.RGB
   Alpha=Saturate(None=Multiply(A=@n8) {constB=0.5})
  )
  n10 = Multiply(A=@n9, B=ScalarParameter() {defaultValue=1, group=DarkNexes_Hexes, parameterName=LumenEmissiveMultiplier})
  FunctionOutput(None=NamedRerouteDeclaration(None=RayTracingQualitySwitch(Normal=@n9, RayTraced=@n10)) {name=Emissive, nodeColor=<rgba 1 0.5859 0 1>})
  NamedRerouteDeclaration(None=Multiply(A=Time(), B=ScalarParameter() {defaultValue=0.3, group=DarkNexes_Hexes, parameterName=HexesIdleDisplacementSpeed})) {name=Time, nodeColor=<rgba 0.9219 0 1 1>}
  TextureCoordinate()
 ]}
