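# Translucent unlit master material for breakable Fortnite glass, with vertex-animation-texture morph nodes, damage functions and an optional crack brightening overlay.
# It fakes lighting by blending lat-long interior/exterior horizon textures, skylight samples, clouds and specular highlights by Fresnel, tinted by time of day.
# Reflections add onto tinted scene color, with simpler mobile and opaque paths.
# Grunge, grime, cubemap and Dark Nexus hex layers use static or quality switches.
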Material /SoberBotany/Materials/M_BreakableGlassVAT_Master
graph {
 properties {
  bAutomaticallySetUsageInEditor=false
  bFullyRough=true
  bUsedWithInstancedStaticMeshes=true
  bUsedWithNanite=true
  blendMode=BLEND_Translucent
  numCustomizedUVs=4
  physMaterial=/Game/FortressPhysicalMaterials/Glass
  pixelDepthOffsetMode=PDOM_Legacy
  shadingModel=MSM_Unlit
  usedShadingModels=Unlit
 }
 nodes [
  n1 = Multiply(A=MaterialFunctionCall() {materialFunction=&CameraDirectionVector}.Result) {constB=-1}
  Abs(None=@n1)
  n2 = CollectionParameter() {
   collection=&FortniteMaterialParameters
   parameterName=SunlightDirectionalVector
  }
  DotProduct(A=CameraVectorWS(), B=@n2)
  n3 = LinearInterpolate(
   A=ScalarParameter() {defaultValue=0.15, group=Opacity, parameterName=Opacity BKP}
   B=ScalarParameter() {defaultValue=0.5, group=Opacity, parameterName=Opacity_Fresnel BKP}
   Alpha=NamedRerouteUsage()
  )
  n4 = VectorParameter() {parameterName=Scalability_OpacityMult, channelNames {a=Epic, b=High, g=Medium, r=Low}, defaultValue=<rgba 1 1 1 1>}
  n5 = Multiply(A=@n3, B=@n4.Low)
  n6 = Add(
   A=Multiply(A=NamedRerouteUsage(), B=ScalarParameter() {defaultValue=0.2, group=Grunge, parameterName=Opacity Grunge Intensity}) {constB=0.3}
   B=MaterialFunctionCall!(
    Low _ PC Low and Med=@n5
    "High _ XB1, PS4"=Multiply(A=@n3, B=@n4.High)
    "Epic _ PC High, XSX, PS5"=Multiply(A=@n3, B=@n4.Epic)
    --------
    "Mobile Low _ PC Perf, Android"=@n5
    Mobile Med _ Switch=Multiply(A=@n3, B=@n4.Medium)
   ) {materialFunction=/Game/Global/MaterialFunctions/TechArt/MF_Fortnite_QualitySwitch_Scalar}.Result
  )
  n7 = LinearInterpolate(
   A=@n6
   B=Multiply(A=@n6, B=ScalarParameter() {defaultValue=1, group=DistanceBlend, parameterName=DistanceBlend_OpacityMult})
   Alpha=NamedRerouteUsage()
  )
  n8 = SetMaterialAttributes!(
   MaterialAttributes
   Base Color=MaterialProxyReplace(Realtime=NamedRerouteUsage(), MaterialProxy=NamedRerouteUsage())
   Opacity=MaterialFunctionCall(
    Low _ PC Low and Med=@n6
    "High _ XB1, PS4"=@n7
    "Epic _ PC High, XSX, PS5"=@n7
    "Mobile Low _ PC Perf, Android"=@n6
    Mobile Med _ Switch=@n6
   ) {materialFunction=&MF_Fortnite_QualitySwitch_Scalar}.Result
   Emissive Color=NamedRerouteUsage()
   Roughness
   Specular
   Normal=Constant3Vector() {constant=<rgba 0 0 1 0>}
  ) {attributeSetTypes [
    #1
    #2
    #3
    #4
    #5
    #6
   ]}
  n9 = GetMaterialAttributes(None=@n8) {attributeGetTypes [#1, #3]}
  n10 = SetMaterialAttributes!(
   MaterialAttributes=@n9.MaterialAttributes
   Base Color=Add(A=@n9.Base Color, B=NamedRerouteUsage())
   Emissive Color=Add(A=@n9.Emissive Color, B=NamedRerouteUsage())
  ) {attributeSetTypes [#1, #3]}
  MaterialFunctionCall!(In=MaterialFunctionCall(Material_In=StaticSwitchParameter(True=@n10, False=@n9.MaterialAttributes) {group=DarkNexus, parameterName=Use Dark Nexus Hexes}) {materialFunction=&MF_FN_DestructibleDamage}.Material_Out) {materialFunction=/Game/Global/Base4Layers_2022/MaterialFunctions/MF_VibrationRattle}
  MaterialFunctionCall!(
   Morph Animation=ScalarParameter() {defaultValue=0.5, group=VAT, parameterName=VATAnim}
   Morph Normal
   Morph Texture=TextureObjectParameter() {group=VAT, parameterName=VATTexture}
   Number of Morph Targets=ScalarParameter() {defaultValue=100, group=VAT, parameterName=NumFrames}
   0-1 Animation Value?
   Morph Target UV
  ) {materialFunction=/Engine/Functions/Engine_MaterialFunctions02/MAXScripts/MS_VertexAnimationTools_MorphTargets}
  n11 = Length(None=ObjectPositionWS() {originType=CameraRelative}) {is=class MaterialExpressionLength}
  Max(A=Subtract(A=@n11, B=ScalarParameter() {defaultValue=10000, parameterName=DistanceBlend_StartDistance})) {constB=0}
  n12 = ScalarParameter() {defaultValue=12000, group=DistanceBlend, parameterName=Distance_BlendDistance}
  n13 = Power(Base=Saturate(None=Divide(A=@n11, B=@n12)), Exp=ScalarParameter() {defaultValue=4, group=DistanceBlend, parameterName=Distance_BlendPow}) {constExponent=4}
  Multiply(A=@n13, B=ScalarParameter() {defaultValue=4, group=DistanceBlend, parameterName=Distance_BlendMult}) {constB=4}
  n14 = TextureSampleParameter2D(UVs=Multiply(A=TextureCoordinate(), B=ScalarParameter() {defaultValue=12, group=Blur, parameterName=BlurNormalTile})) {
   group=Blur
   parameterName=BlurNormalMap
   samplerType=SAMPLERTYPE_Normal
   texture=/Game/Global/Textures/Default/Tileable_Patterns/T_SkinBump_N
  }
  n15 = ComponentMask(None=@n14.RGB) {g=true, r=true}
  n16 = ScalarParameter() {defaultValue=0.1, group=Blur, parameterName=BlurAmount, sortPriority=19}
  NamedRerouteDeclaration(None=SceneColor!(OffsetFraction=Multiply(A=@n15, B=@n16)) {inputMode=OffsetFraction}.RGB) {name=DistortedSceneColor, nodeColor=<rgba 0.4062 1 0 1>}
  n17 = ScalarParameter() {desc=Adjust the value for the black part of grime mask, group=Grime, parameterName=Grime_AdjustMin}
  n18 = ScalarParameter() {defaultValue=0.8, desc=Adjust the value for the white part of grime mask, group=Grime, parameterName=Grime_AdjustMax}
  n19 = ScalarParameter() {defaultValue=1, group=Grime, parameterName=Grime_RepeatU}
  n20 = ScalarParameter() {defaultValue=1, group=Grime, parameterName=Grime_RepeatV}
  n21 = TextureSampleParameter2D(UVs=Multiply(A=TextureCoordinate(), B=AppendVector(A=@n19, B=@n20))) {
   group=Grime
   parameterName=TextureMask
   samplerType=SAMPLERTYPE_Masks
   texture=/Game/Global/Materials/EnvironmentGlass/T_GlassTestMask_Indestruct
   channelNames {a="", b=Mask, g=Roughness, r=Grime}
  }
  n22 = StaticSwitchParameter(True=OneMinus(None=@n21.Grime), False=@n21.Grime) {desc="Inverts the grime mask in Red channel of TextureMask sample.  ", group=Grime, parameterName=InvertGrime}
  NamedRerouteDeclaration(None=Saturate(None=LinearInterpolate(A=@n17, B=@n18, Alpha=@n22) {constA=1})) {name=GrimeMask, nodeColor=<rgba 0.4583 0.2142 0.0789 1>}
  n23 = Divide(A=@n11, B=ScalarParameter() {defaultValue=7000, group=DistanceBlend, parameterName=CubeEmissive_Distance_BlendDistance})
  NamedRerouteDeclaration(None=Power(Base=Saturate(None=@n23), Exp=ScalarParameter() {defaultValue=4, group=DistanceBlend, parameterName=CubeEmissive_Distance_BlendPow}) {constExponent=4}) {name=DistantBlendEmissive, nodeColor=<rgba 0 1 0.1094 1>}
  n24 = Power(Base=NamedRerouteUsage(), Exp=ScalarParameter() {defaultValue=1, group=CubeMap, parameterName=CubemapEmissivePow})
  n25 = LinearInterpolate(
   A=ScalarParameter() {group=CubeMap, parameterName=CubemapEmissiveFraction}
   B=ScalarParameter() {group=CubeMap, parameterName=CubemapEmissiveFraction_Distant}
   Alpha=NamedRerouteUsage()
  )
  NamedRerouteDeclaration(None=LinearInterpolate(A=Multiply(A=@n24, B=@n25), Alpha=NamedRerouteUsage()) {constB=0}) {name=GlassEmissive, nodeColor=<rgba 0.6406 1 0 1>}
  n26 = Constant()
  n27 = ScalarParameter() {
   defaultValue=0.5
   desc="A multiplier on the cubemap to blend it in and out.  Higher values show it more."
   group=CubeMap
   parameterName=CubeMap_Mult
  }
  n28 = ScalarParameter() {
   defaultValue=1
   desc=When the angle of the surface faces more away from the camera this will be the value used for Cubemap mult.
   group=CubeMap
   parameterName=CubeMap_Fresnel
  }
  n29 = ReflectionVectorWS()
  n30 = TextureSampleParameterCube!(UVs=@n29, Apply View MipBias) {
   group=CubeMap
   parameterName=CubemapTexture
   samplerType=SAMPLERTYPE_LinearColor
   texture=/Game/Creative/Environments/Meshes/Glass/Textures/T_Glass_Cubemap_01
  }
  n31 = SkyLightEnvMapSample(
   Direction=@n29
   Roughness=LinearInterpolate(
    A=ScalarParameter() {defaultValue=0.1, group=CubeMap, parameterName=SkylightEnvMapRoughness}
    B=ScalarParameter() {defaultValue=1, group=Grunge, parameterName=Grunge_Roughness}
    Alpha=NamedRerouteUsage()
   )
  )
  n32 = VectorParameter() {group=CubeMap, parameterName=SkylightEnvMapDesat_TOD, channelNames {a=Night, b=Evening, g=Day, r=Morning}}
  n33 = MaterialFunctionCall(Value_Morning=@n32.Morning, Value_Day=@n32.Morning, Value_Evening=@n32.Morning, Value_Night=@n32.Morning) {materialFunction=&Scalar_Time_of_Day}
  n34 = StaticSwitchParameter(
   True=TextureSampleParameter2D(UVs=ComponentMask(None=@n29) {g=true, r=true}) {group=CubeMap, parameterName=CubemapTexture2d, texture=/Game/Packages/Fortress_SharedMaterials/Textures/T_Cubemap_01_tex}.RGB
   False=StaticSwitchParameter(True=@n30.RGB, False=Desaturation(None=@n31, Fraction=@n33.Result)) {defaultValue=true, group=CubeMap, parameterName=UseTextureCubeMap}
  ) {group=CubeMap, parameterName=UseTextureCubeMap2d}
  n35 = ScalarParameter() {defaultValue=1, group=CubeMap, parameterName=CubePow}
  n36 = Multiply(
   A=LinearInterpolate(A=@n27, B=@n28, Alpha=NamedRerouteUsage())
   B=Multiply(A=Power(Base=@n34, Exp=@n35), B=ScalarParameter() {defaultValue=1, group=CubeMap, parameterName=CubeIntensity})
  )
  n37 = VectorParameter() {group=CubeMap, parameterName=CubemapColorScale, defaultValue=<rgba 1 1 1 0>}
  n38 = StaticSwitchParameter(
   True=EyeAdaptationInverse(LightValueInput=Multiply(A=@n36, B=@n37.RGB), AlphaInput=ScalarParameter() {defaultValue=0.3, parameterName=ExposureCompensation}).EyeAdaptationInverse
   False=@n26
  ) {defaultValue=true, group=CubeMap, parameterName=UseCubeMap}
  NamedRerouteDeclaration(None=MaterialFunctionCall!(
    Low _ PC Low and Med=@n26
    "High _ XB1, PS4"=@n38
    "Epic _ PC High, XSX, PS5"=@n38
    --------
    "Mobile Low _ PC Perf, Android"=@n26
    Mobile Med _ Switch=@n38
   ) {materialFunction=/Game/Global/MaterialFunctions/TechArt/MF_Fortnite_QualitySwitch_Vector3}.Result) {name=CubeMapColor, nodeColor=<rgba 0.2422 0 1 1>}
  n39 = CollectionParameter() {
   collection=&FortniteMaterialParameters
   parameterName=FortniteActiveTimeOfDay
  }
  NamedRerouteDeclaration(None=Saturate(None=MaterialFunctionCall(Float4=@n39) {materialFunction=&BreakOutFloat4Components}.A)) {name=NightTime, nodeColor=<rgba 0 1 0.6016 1>}
  n40 = NamedRerouteUsage()
  n41 = StaticSwitchParameter(
   True=MaterialFunctionCall(
    Color_Morning=VectorParameter() {group=Glass Color, parameterName=GlassColorMorning, defaultValue=<rgba 0.3712 0.5035 0.6198 1>}.RGB
    Color_Day=VectorParameter() {group=Glass Color, parameterName=GlassColorDay, defaultValue=<rgba 0.3712 0.5035 0.6198 1>}.RGB
    Color_Evening=VectorParameter() {group=Glass Color, parameterName=GlassColorEvening, defaultValue=<rgba 0.3712 0.5035 0.6198 1>}.RGB
    Color_Night=VectorParameter() {group=Glass Color, parameterName=GlassColorNight, defaultValue=<rgba 0.3712 0.5035 0.6198 1>}.RGB
   ) {materialFunction=&Color_Time_of_Day}.Result
   False=VectorParameter() {group=Glass Color, parameterName=GlassColor, defaultValue=<rgba 0.3712 0.5035 0.6198 1>}.RGB
  ) {group=Glass Color, parameterName=UseTODGlassColor}
  n42 = VectorParameter() {parameterName=Scalability_BaseColorMult, channelNames {a=Epic, b=High, g=Medium, r=Low}, defaultValue=<rgba 1 1 1 1>}
  n43 = Multiply(A=@n41, B=@n42.Low)
  n44 = MaterialFunctionCall(
   Low _ PC Low and Med=@n43
   "High _ XB1, PS4"=Multiply(A=@n41, B=@n42.High)
   "Epic _ PC High, XSX, PS5"=Multiply(A=@n41, B=@n42.Epic)
   "Mobile Low _ PC Perf, Android"=@n43
   Mobile Med _ Switch=Multiply(A=@n41, B=@n42.Medium)
  ) {materialFunction=&MF_Fortnite_QualitySwitch_Vector3}
  n45 = ScalarParameter() {defaultValue=1, group=Glass Color, parameterName=GlassColorScale}
  n46 = LinearInterpolate(
   A=NamedRerouteDeclaration(None=Multiply(A=@n44.Result, B=@n45)) {name=GlassColor, nodeColor=<rgba 0 1 0.8359 1>}
   B=VectorParameter() {group=Grunge, parameterName=GrungeColor, defaultValue=<rgba 0.5156 0.2683 0.0785 1>}.RGB
   Alpha=NamedRerouteUsage()
  )
  n47 = Saturate(None=Sign(None=ScalarParameter() {bUseCustomPrimitiveData=true, parameterName=DisableFakeShading}))
  NamedRerouteDeclaration(None=LinearInterpolate(A=Add(A=@n40, B=@n46), B=@n46, Alpha=NamedRerouteDeclaration(None=@n47) {name=Fake Shading Switch, nodeColor=<rgba 0 1 0.1328 1>})) {name=GlassColorWithMask, nodeColor=<rgba 1 0.0938 0 1>}
  NamedRerouteDeclaration(None=MaterialFunctionCall() {materialFunction=&/MF_HexagonPattern}.Result) {name=Square Color glitch, nodeColor=<rgba 0.2422 1 0 1>}
  n48 = Multiply(A=NamedRerouteUsage(), B=LinearInterpolate(Alpha=NamedRerouteUsage()) {constA=0.1, constB=0.4}) {constB=0.1}
  n49 = NamedRerouteUsage()
  n50 = Constant() {r=1}
  n51 = ScalarParameter() {defaultValue=530, group=DarkNexus, parameterName=Depth gradient, sortPriority=50}
  n52 = LinearInterpolate(
   A=Multiply(A=ScalarParameter() {defaultValue=0.2, group=DarkNexus, parameterName=Depth gradient Color Multiply, sortPriority=50}, B=@n49)
   B=@n49
   Alpha=Saturate(None=DepthFade(Opacity=@n50, FadeDistance=@n51))
  )
  n53 = LinearInterpolate(Alpha=NamedRerouteUsage()) {constA=0.35, constB=0.2}
  n54 = NamedRerouteUsage()
  n55 = Multiply(
   A=NamedRerouteUsage()
   B=Multiply(
    A=NamedRerouteDeclaration(None=@n52) {name=Depth Gradient, nodeColor=<rgba 0 0.2734 1 1>}
    B=LinearInterpolate(B=@n53, Alpha=@n54) {constA=0.2, constB=0.5}
   )
  )
  n56 = Multiply(A=NamedRerouteUsage()) {constB=2}
  n57 = Multiply(A=NamedRerouteUsage()) {constB=0.5}
  n58 = LinearInterpolate(A=Multiply(A=@n55) {constB=0.15}, B=@n55, Alpha=Saturate(None=Add(A=@n56, B=@n57))) {constB=0}
  n59 = MaterialFunctionCall(
   Low _ PC Low and Med=@n48
   "High _ XB1, PS4"=@n48
   "Epic _ PC High, XSX, PS5"=QualitySwitch(Default=@n58, Low=@n48, High=@n58, Medium=@n48, Epic=@n58)
   "Mobile Low _ PC Perf, Android"=@n48
   Mobile Med _ Switch=@n48
  ) {materialFunction=&MF_Fortnite_QualitySwitch_Vector3}
  n60 = ScalarParameter() {defaultValue=0.2, group=DarkNexus, parameterName=HexesEmissiveIntensity}
  NamedRerouteDeclaration(None=EyeAdaptationInverse(LightValueInput=Multiply(A=@n59.Result, B=@n60), AlphaInput=Constant() {r=1}).EyeAdaptationInverse) {name=DarkNexusEmissive, nodeColor=<rgba 0.0547 0 1 1>}
  n61 = ScalarParameter() {defaultValue=1, group=DarkNexus, parameterName=HexesBaseColorIntensity}
  NamedRerouteDeclaration(None=Saturate(None=Multiply(A=@n59.Result, B=@n61))) {name=DarkNexusBaseColor, nodeColor=<rgba 0.0547 0 1 1>}
  n62 = VertexNormalWS()
  n63 = CameraVectorWS()
  n64 = OneMinus(None=Abs(None=DotProduct(A=@n62, B=@n63)))
  NamedRerouteDeclaration(None=Multiply(A=@n64, B=@n64)) {name=Fresnel, nodeColor=<rgba 1 0 0.0703 1>}
  n65 = Divide(A=PixelDepth().R) {constB=5000}
  n66 = Multiply(A=@n65, B=@n65)
  NamedRerouteDeclaration(None=VertexInterpolator(VS=Saturate(None=@n66)).PS) {name=Distance Mask, nodeColor=<rgba 0.9219 0 1 1>}
  n67 = Multiply(A=NamedRerouteUsage(), B=ScalarParameter() {defaultValue=0.01, group=Normal, parameterName=Normal Intensity, sortPriority=1}) {constB=0.1}
  n68 = VertexNormalWS()
  NamedRerouteDeclaration(None=ReflectionVectorWS(CustomWorldNormal=Add(A=@n67, B=@n68))) {name=Reflection Vector With Normals, nodeColor=<rgba 1 0.2578 0 1>}
  n69 = VectorParameter() {group=Specular, parameterName=Specular Color Morning, sortPriority=2, defaultValue=<rgba 2 0.9085 0.678 1>}
  n70 = VectorParameter() {group=Specular, parameterName=Specular Color Day, sortPriority=3, defaultValue=<rgba 1.4 1.2537 1.0938 1>}
  n71 = VectorParameter() {group=Specular, parameterName=Specular Color Evening, sortPriority=4, defaultValue=<rgba 2 0.7588 0.5208 1>}
  n72 = VectorParameter() {group=Specular, parameterName=Specular Color Night, sortPriority=5, defaultValue=<rgba 0.5391 0.7147 1.5 1>}
  NamedRerouteDeclaration(None=VertexInterpolator(VS=MaterialFunctionCall(Color_Morning=@n69.RGB, Color_Day=@n70.RGB, Color_Evening=@n71.RGB, Color_Night=@n72.RGB) {materialFunction=&Color_Time_of_Day}.Result).PS) {name=Specular Color, nodeColor=<rgba 0.5 0 1 1>}
  MaterialExpressionAbs_5 = Abs(None=VertexNormalWS())
  n73 = Convert(None=@MaterialExpressionAbs_5) {
   nodeName=Break Float3
   convertInputs [{
     type=Vector3
     expressionInput {
      expression=&self:MaterialExpressionAbs_5
      inputName=None
      mask=0
      maskA=0
      maskB=0
      maskG=0
      maskR=0
      outputIndex=0
     }
    }]
   convertOutputs [{type=Scalar}, {type=Scalar}, {type=Scalar}]
  }
  n74 = Add(A=@n73) {constB=0.2}
  n75 = Add(A=@n73) {constB=-0.2}
  n76 = AppendVector(A=Subtract(A=@n74, B=Max(A=@n73, B=@n75)), B=Subtract(A=@n73, B=Max(A=@n74, B=@n75)))
  NamedRerouteDeclaration(None=Saturate(None=Ceil(None=@n76))) {name=Triplanar Mask, nodeColor=<rgba 1 0 0 1>}
  n77 = VectorParameter() {group=Texture Pattern, parameterName=Texture Pattern Color 1, sortPriority=1, defaultValue=<rgba 0.5156 0.0645 0.1063 1>}
  n78 = VectorParameter() {group=Texture Pattern, parameterName=Texture Pattern Color 2, sortPriority=1, defaultValue=<rgba 0.5156 0.0645 0.1063 1>}
  n79 = WorldPosition() {worldPositionShaderOffset=WPT_CameraRelative}
  n80 = ScalarParameter() {defaultValue=13, group=Texture Pattern, parameterName=Texture Pattern Scale, sortPriority=0}
  n81 = VertexInterpolator(VS=Divide(
    A=TransformPosition!(None=@n79.XYZ, Periodic World Tile Size=@n80) {transformSourceType=TRANSFORMPOSSOURCE_TranslatedWorld, transformType=TRANSFORMPOSSOURCE_PeriodicWorld}
    B=@n80
   ))
  n82 = ComponentMask(None=@n81.PS) {g=true, r=true}
  n83 = ComponentMask(None=@n81.PS) {b=true, r=true}
  n84 = ComponentMask(None=NamedRerouteUsage()) {g=true}
  n85 = NamedRerouteUsage()
  n86 = TextureSampleParameter2D(UVs=LinearInterpolate(A=LinearInterpolate(A=@n82, B=@n83, Alpha=@n84), B=ComponentMask(None=@n81.PS) {b=true, g=true}, Alpha=ComponentMask(None=@n85) {r=true})) {
   group=Texture Pattern
   parameterName=Texture Pattern
   samplerType=SAMPLERTYPE_LinearGrayscale
   texture=&/T_BreakableGlass_Pattern_A
  }
  NamedRerouteDeclaration(None=LinearInterpolate(A=@n77.RGB, B=@n78.RGB, Alpha=@n86.R)) {name=Pattern Color, nodeColor=<rgba 1 0.3984 0 1>}
  NamedRerouteDeclaration(None=LinearInterpolate(
    A=ScalarParameter() {defaultValue=0.01, group=Texture Pattern, parameterName=Texture Pattern Opacity Low, sortPriority=2}
    B=ScalarParameter() {defaultValue=0.04, group=Texture Pattern, parameterName=Texture Pattern Opacity High, sortPriority=3}
    Alpha=@n86.R
   )) {name=Pattern Opacity, nodeColor=<rgba 0.0781 0 1 1>}
  n87 = Saturate(None=Subtract(A=WorldPosition().Z, B=ScalarParameter() {defaultValue=-1000, parameterName=Interior Reflection Height}))
  NamedRerouteDeclaration(None=Max(A=OneMinus(None=@n87), B=ScalarParameter() {bUseCustomPrimitiveData=true, parameterName=Interior Reflection, primitiveDataIndex=11})) {name=Interior Mask, nodeColor=<rgba 0 1 0.8828 1>}
  n88 = Constant()
  n89 = MaterialFunctionCall() {materialFunction=/Game/Effects/Niagara/MasterMaterials/Functions/InstanceBounds_0-1_UVW}
  n90 = Constant() {r=0.5}
  n91 = ScalarParameter() {defaultValue=0.75, group=Grunge, parameterName=Grunge_CenterMask_Size}
  n92 = ScalarParameter() {defaultValue=0.8, group=Grunge, parameterName=Grunge_CenterMask_Hardness}
  n93 = StaticBool() {value=true}
  n94 = TextureObjectParameter() {parameterName=GrungeTexture, samplerType=SAMPLERTYPE_Masks, texture=&/T_Glass_Noise}
  n95 = ScalarParameter() {defaultValue=167.7897, group=Grunge, parameterName=Grunge Scale}
  n96 = Add(A=ActorPositionWS(), B=WorldPosition().XYZ)
  n97 = Power(
   Base=StaticComponentMaskParameter(None=MaterialFunctionCall(Export Float 4=@n93, TextureObject=@n94, TextureSize=@n95, WorldPosition=@n96) {materialFunction=/Game/Weapons/WeaponSkins/Wraps/MaterialFunctions/WorldAlignedTexture}.XYZ Texture) {defaultR=true, group=Grunge, parameterName=GrungeChannel}
   Exp=ScalarParameter() {defaultValue=1, group=Grunge, parameterName=Grunge_Power}
  )
  n98 = ScalarParameter() {defaultValue=1, group=Grunge, parameterName=Grunge_Mult}
  n99 = Multiply(
   A=LinearInterpolate(
    A=ScalarParameter() {defaultValue=1, group=Grunge, parameterName=Grunge_CenterMask_Intensity_Edges}
    B=ScalarParameter() {defaultValue=0.1, group=Grunge, parameterName=Grunge_CenterMask_Intensity_Center}
    Alpha=SphereMask(A=@n89.RGB, B=@n90, Radius=@n91, Hardness=@n92) {hardnessPercent=0}
   )
   B=Saturate(None=Multiply(A=@n97, B=@n98))
  )
  NamedRerouteDeclaration(None=MaterialFunctionCall(
    Low _ PC Low and Med=@n88
    "High _ XB1, PS4"=@n99
    "Epic _ PC High, XSX, PS5"=@n99
    "Mobile Low _ PC Perf, Android"=@n88
    Mobile Med _ Switch=@n88
   ) {materialFunction=&MF_Fortnite_QualitySwitch_Scalar}.Result) {name=GrungeMask, nodeColor=<rgba 0.602 0.7917 0.7205 1>}
  n100 = WorldPosition() {worldPositionShaderOffset=WPT_CameraRelative}
  n101 = ScalarParameter() {defaultValue=300, group=Mask, parameterName=Mask Scale, sortPriority=1}
  n102 = VertexInterpolator(VS=Divide(
    A=TransformPosition(None=@n100.XYZ, Periodic World Tile Size=@n101) {transformSourceType=TRANSFORMPOSSOURCE_TranslatedWorld, transformType=TRANSFORMPOSSOURCE_PeriodicWorld}
    B=@n101
   ))
  n103 = NamedRerouteUsage()
  n104 = LinearInterpolate(
   A=LinearInterpolate(A=ComponentMask(None=@n102.PS) {g=true, r=true}, B=ComponentMask(None=@n102.PS) {b=true, r=true}, Alpha=ComponentMask(None=@n103) {g=true})
   B=ComponentMask(None=@n102.PS) {b=true, g=true}
   Alpha=ComponentMask(None=NamedRerouteUsage()) {r=true}
  )
  n105 = TextureSampleParameter2D(UVs=@n104) {group=Mask, parameterName=Mask, samplerType=SAMPLERTYPE_Masks, sortPriority=0, texture=&/T_GlassReflection_M}
  NamedRerouteDeclaration(None=LinearInterpolate(A=@n105.R, Alpha=NamedRerouteUsage()) {constB=0.97}) {name=Specular Mask, nodeColor=<rgba 1 0.5156 0 1>}
  NamedRerouteDeclaration(None=MaterialFunctionCall() {materialFunction=&Color_Time_of_Day}.Result) {name=Color of day, nodeColor=<rgba 0 0.7656 1 1>}
  n106 = TextureSampleParameter2D(UVs=@n104) {group=Normal, parameterName=Normal, samplerType=SAMPLERTYPE_LinearColor, sortPriority=0, texture=&/T_GlassReflection_N}
  n107 = Add(A=@n106.RGB) {constB=-0.5}
  n108 = NamedRerouteUsage()
  NamedRerouteDeclaration(None=ShadingPathSwitch(Default=LinearInterpolate(A=@n107, Alpha=@n108) {constB=0}, Mobile=Constant())) {name=Normals, nodeColor=<rgba 1 0.8438 0 1>}
  n109 = Subtract(A=@n75, B=Max(A=@n74, B=@n73))
  n110 = DotProduct(A=Saturate(None=AppendVector(A=@n76, B=@n109)), B=Constant() {r=1})
  NamedRerouteDeclaration(None=Power(Base=OneMinus(None=@n110)) {constExponent=32}) {name=Seam Mask, nodeColor=<rgba 0 0.5078 1 1>}
  n111 = VectorParameter() {group=Sky, parameterName=Sky Color Morning, sortPriority=4, defaultValue=<rgba 0.4062 0.3398 0.3281 1>}
  n112 = VectorParameter() {group=Sky, parameterName=Sky Color Day, sortPriority=5, defaultValue=<rgba 0.3525 0.4219 0.3502 1>}
  n113 = VectorParameter() {group=Sky, parameterName=Sky Color Evening, sortPriority=6, defaultValue=<rgba 0.4271 0.3328 0.3288 1>}
  n114 = VectorParameter() {group=Sky, parameterName=Sky Color Night, sortPriority=7, defaultValue=<rgba 0.4844 0.4844 0.4844 1>}
  NamedRerouteDeclaration(None=VertexInterpolator(VS=MaterialFunctionCall(Color_Morning=@n111.RGB, Color_Day=@n112.RGB, Color_Evening=@n113.RGB, Color_Night=@n114.RGB) {materialFunction=&Color_Time_of_Day}.Result).PS) {name=Sky Color, nodeColor=<rgba 0.2422 0 1 1>}
  n115 = VectorParameter() {group=Reflection, parameterName=Reflection Color Morning, sortPriority=3, defaultValue=<rgba 1.6 0.77 0.6474 1>}
  n116 = VectorParameter() {group=Reflection, parameterName=Reflection Color Day, sortPriority=4, defaultValue=<rgba 1 1 1 1>}
  n117 = VectorParameter() {group=Reflection, parameterName=Reflection Color Evening, sortPriority=5, defaultValue=<rgba 1.4 0.7821 0.6612 1>}
  n118 = VectorParameter() {group=Reflection, parameterName=Reflection Color Night, sortPriority=6, defaultValue=<rgba 0.3061 0.4715 0.6875 1>}
  NamedRerouteDeclaration(None=VertexInterpolator(VS=MaterialFunctionCall(Color_Morning=@n115.RGB, Color_Day=@n116.RGB, Color_Evening=@n117.RGB, Color_Night=@n118.RGB) {materialFunction=&Color_Time_of_Day}.Result).PS) {name=Reflection Color, nodeColor=<rgba 1 0 0.3984 1>}
  n119 = ComponentMask(None=@n2) {b=true, g=true, r=true}
  n120 = Abs(None=VertexNormalWS())
  n121 = DotProduct(A=@n119, B=@n120)
  n122 = Subtract(A=Multiply(A=Multiply(A=@n121) {constB=2}, B=@n120), B=@n119)
  n123 = Max(A=DotProduct(A=Normalize(VectorInput=@n122), B=@n1)) {constB=0}
  n124 = ScalarParameter() {defaultValue=4, parameterName=Fake Phong}
  n125 = Multiply(A=NamedRerouteUsage(), B=Multiply(A=Power(Base=@n123, Exp=@n124), B=ScalarParameter() {defaultValue=1.2, parameterName=Fake Specular Mult}))
  NamedRerouteDeclaration(None=Saturate(None=Add(A=@n125) {constB=0.3})) {name=Fake Specular, nodeColor=<rgba 0 1 0.9063 1>}
  n126 = LinearInterpolate(
   A=ScalarParameter() {defaultValue=4, group=Sky, parameterName=Sky Mip Bias High, sortPriority=3}
   B=ScalarParameter() {group=Sky, parameterName=Sky Mip Bias Low, sortPriority=2}
   Alpha=NamedRerouteUsage()
  )
  n127 = Desaturation(
   None=SkyLightEnvMapSample(Direction=NamedRerouteUsage(), Roughness=Saturate(None=@n126))
   Fraction=Multiply(A=ScalarParameter() {group=Sky, parameterName=Sky Saturation Compensation, sortPriority=0}) {constB=-1}
  )
  n128 = ComponentMask(None=VertexNormalWS()) {b=true}
  n129 = OneMinus(None=Multiply(A=ScalarParameter() {defaultValue=0.5, group=Sky, parameterName=Sky Top Reflection Suppress, sortPriority=1}, B=Abs(None=@n128)) {constB=0.7})
  NamedRerouteDeclaration(None=Multiply(A=Multiply(A=@n127, B=@n129), B=NamedRerouteUsage())) {name=Sky, nodeColor=<rgba 0 0.7188 1 1>}
  n130 = NamedRerouteUsage()
  n131 = ComponentMask(None=@n130) {g=true, r=true}
  n132 = ComponentMask(None=@n130) {b=true}
  n133 = LinearInterpolate(Alpha=Abs(None=NamedRerouteDeclaration(None=@n132) {name=Ref B, nodeColor=<rgba 0.0781 1 0 1>})) {constA=0.3}
  n134 = TextureSample(UVs=Multiply(A=Divide(A=@n131, B=@n133), B=ScalarParameter() {defaultValue=0.4, group=Clouds, parameterName=Clouds Tiling}) {constB=0.5}) {samplerType=SAMPLERTYPE_LinearColor, texture=/Game/TimeOfDay/TODM_Sequence/Materials/Sky/Textures/T_Ch6_OvercastClouds}
  NamedRerouteDeclaration(None=Multiply(A=@n134.R, B=NamedRerouteUsage())) {name=Clouds RGB, nodeColor=<rgba 1 0.9609 0 1>}
  n135 = Sign(None=@n132)
  n136 = SmoothStep(
   Min=ScalarParameter() {defaultValue=0.544, group=Clouds, parameterName=Clouds Range Min}
   Max=ScalarParameter() {defaultValue=0.8316, group=Clouds, parameterName=Clouds Range Max}
   Value=Multiply(A=@n134.B, B=Saturate(None=@n135))
  )
  NamedRerouteDeclaration(None=Multiply(A=Saturate(None=@n136), B=ScalarParameter() {defaultValue=0.7, group=Clouds, parameterName=Clouds Opacity})) {name=Clouds A, nodeColor=<rgba 0.6406 0 1 1>}
  n137 = NamedRerouteUsage()
  n138 = Arctangent2Fast(Y=ComponentMask(None=@n137) {g=true}, X=ComponentMask(None=@n137) {r=true})
  n139 = MaterialFunctionCall(Multiplier=Constant() {r=2}) {materialFunction=&Pi}
  n140 = Multiply(
   A=Frac(None=Divide(A=@n138, B=@n139.Result))
   B=ScalarParameter() {defaultValue=4, group=Reflection, parameterName=Reflection Tiling X, sortPriority=1}
  ) {constB=4}
  n141 = Multiply(
   A=ArcsineFast!(None=ComponentMask(None=@n137) {b=true})
   B=Divide(B=MaterialFunctionCall() {materialFunction=&Pi}.Result) {constA=2}
  )
  n142 = Add(A=Multiply(A=Multiply(A=@n141) {constB=-1}, B=ScalarParameter() {defaultValue=2, group=Reflection, parameterName=Reflection Tiling Y, sortPriority=2}) {constB=2}) {constB=0.5}
  n143 = AppendVector(A=@n140, B=@n142)
  n144 = Add(A=Frac(None=@n140)) {constB=-0.5}
  n145 = ScalarParameter() {defaultValue=4, group=Reflection, parameterName=Reflection Tiling X, sortPriority=1}
  n146 = AppendVector(A=Multiply(A=Abs(None=@n144), B=Divide(B=@n145)), B=@n142)
  n147 = DDX(Value=@n146)
  n148 = DDY(Value=@n146)
  n149 = TextureSampleParameter2D!(UVs=@n143, "DDX(UVs)"=@n147, "DDY(UVs)"=@n148, Apply View MipBias) {mipValueMode=TMVM_Derivative, parameterName=Reflection Exterior, texture=&/T_BreakableGlass_Reflection_Hera_Exterior}
  n150 = TextureSampleParameter2D(UVs=@n143, "DDX(UVs)"=@n147, "DDY(UVs)"=@n148) {mipValueMode=TMVM_Derivative, parameterName=Reflection Interior, texture=&/T_BreakableGlass_Reflection_Hera_Interior}
  n151 = LinearInterpolate(A=@n149.RGBA, B=@n150.RGBA, Alpha=NamedRerouteUsage())
  NamedRerouteDeclaration(None=Multiply(A=ComponentMask(None=@n151) {b=true, g=true, r=true}, B=NamedRerouteUsage())) {name=Horizon Reflection RGB, nodeColor=<rgba 0 1 0.3203 1>}
  n152 = LinearInterpolate(
   A=ScalarParameter() {defaultValue=-3.5, group=Reflection, parameterName=Reflection Breakup Min}
   B=ScalarParameter() {defaultValue=1.2, group=Reflection, parameterName=Reflection Breakup Max}
   Alpha=NamedRerouteUsage()
  ) {constA=3, constB=0.7}
  NamedRerouteDeclaration(None=Multiply(A=ComponentMask(None=@n151) {a=true}, B=Saturate(None=@n152))) {name=Horizon Reflection A, nodeColor=<rgba 1 0 0 1>}
  n153 = SceneColor!(UVs)
  n154 = ScalarParameter() {defaultValue=0.6, group=SceneColor, parameterName=SceneColor Highlight Suppress}
  n155 = ScalarParameter() {group=SceneColor, parameterName=SceneColor Highlight Low}
  n156 = ScalarParameter() {defaultValue=0.2, group=SceneColor, parameterName=SceneColor Highlight High}
  n157 = Desaturation(None=@n153.RGB, Fraction=Constant() {r=1})
  n158 = Multiply(
   A=LinearInterpolate(A=@n153.RGB, B=Multiply(A=@n153.RGB, B=@n154) {constB=0.5}, Alpha=SmoothStep(Min=@n155, Max=@n156, Value=@n157) {constMax=0.7, constMin=0.2})
   B=VectorParameter() {group=SceneColor, parameterName=SceneColor Color, defaultValue=<rgba 0.4375 1.1783 1.2 1>}.RGB
  )
  n159 = LinearInterpolate(A=@n158, B=Multiply(A=@n158, B=@n158), Alpha=NamedRerouteUsage())
  n160 = EyeAdaptationInverse(LightValueInput=NamedRerouteUsage(), AlphaInput=Constant() {r=1})
  n161 = NamedRerouteUsage()
  NamedRerouteDeclaration(None=StaticSwitchParameter(True=LinearInterpolate(A=@n159, B=@n160.EyeAdaptationInverse, Alpha=@n161), False=@n159) {group=Texture Pattern, parameterName=Enable Texture Pattern, sortPriority=0}) {name=Scene Color, nodeColor=<rgba 1 0.4688 0 1>}
  n162 = Add(A=SkyAtmosphereLightDirection(), B=CameraVectorWS())
  n163 = Multiply(A=NamedRerouteUsage(), B=ScalarParameter() {defaultValue=0.0012, group=Specular, parameterName=Specular Normal Intensity, sortPriority=1}) {constB=0.1}
  n164 = PixelNormalWS()
  n165 = LinearInterpolate(Alpha=NamedRerouteUsage()) {constA=20, constB=0.5}
  n166 = Power(
   Base=DotProduct(A=Normalize(VectorInput=@n162), B=Add(A=@n163, B=@n164))
   Exp=Multiply(A=Saturate(None=@n165), B=ScalarParameter() {defaultValue=300, group=Specular, parameterName=Specular Falloff, sortPriority=0})
  ) {constExponent=300}
  n167 = NamedRerouteUsage()
  n168 = NamedRerouteUsage()
  NamedRerouteDeclaration(None=Multiply(A=Multiply(A=@n166, B=@n167), B=OneMinus(None=@n168))) {name=Specular Highlight, nodeColor=<rgba 1 0 0.1641 1>}
  n169 = ScalarParameter() {
   defaultValue=1
   desc="Tighten/loosen the fresnel  blend.  Fresnel allows for different values of opacity and roughness based on the angle of the surface facing the camera. (Glass is shinier when looking at it from the edge)"
   parameterName=Fresnel Exponent
  }
  n170 = ScalarParameter() {
   defaultValue=0.1
   desc="Tighten/loosen the fresnel  blend.  Fresnel allows for different values of opacity and roughness based on the angle of the surface facing the camera. (Glass is shinier when looking at it from the edge)"
   parameterName=Fresnel Base
  }
  NamedRerouteDeclaration(None=Fresnel(ExponentIn=@n169, BaseReflectFractionIn=@n170, Normal=PixelNormalWS()) {baseReflectFraction=0}) {name=FresnelMask, nodeColor=<rgba 1 0.8814 0 1>}
  NamedRerouteDeclaration(None=@n123) {name=Fake Spec Highlight, nodeColor=<rgba 0.4766 1 0 1>}
  n171 = NamedRerouteUsage()
  NamedRerouteDeclaration(None=@n171) {name=GlassColorFinal, nodeColor=<rgba 1 0 0.9375 1>}
  NamedRerouteDeclaration(None=@n13) {name=DistantBlend, nodeColor=<rgba 1 0.9844 0 1>}
  PixelNormalWS()
  ShadingPathSwitch(
   Default=ScalarParameter() {defaultValue=0.6, group=SceneColor, parameterName=SceneColor Opacity}
   Mobile=ScalarParameter() {defaultValue=0.6, group=Mobile, parameterName=Mobile Opacity}
  )
  StaticSwitchParameter(
   True=LinearInterpolate(
    A=NamedRerouteUsage()
    B=VectorParameter() {group=Grime, parameterName=GrimeColor, defaultValue=<rgba 0.3712 0.5035 0.6198 1>}.RGB
    Alpha=NamedRerouteUsage()
   )
   False=@n171
  ) {desc="Grime is typically for dirt.  Grim has separate color and roughness", group=Grime, parameterName=UseGrimeTexture}
  n172 = NamedRerouteUsage()
  n173 = NamedRerouteUsage()
  n174 = NamedRerouteUsage()
  n175 = NamedRerouteUsage()
  n176 = Multiply(
   A=LinearInterpolate(A=LinearInterpolate(A=@n173, B=@n174, Alpha=@n175), B=NamedRerouteUsage(), Alpha=NamedRerouteUsage())
   B=LinearInterpolate(
    A=ScalarParameter() {defaultValue=0.2, group=Fresnel, parameterName=Fresnel Reflection Intensity}
    B=ScalarParameter() {defaultValue=1.4, group=Fresnel, parameterName=Fresnel Reflection Intensity 90º}
    Alpha=NamedRerouteUsage()
   )
  ) {constB=0.2}
  n177 = Add(A=@n176, B=NamedRerouteUsage())
  n178 = NamedRerouteDeclaration(None=Multiply(A=ShadingPathSwitch(Default=@n177, Mobile=@n176), B=NamedRerouteUsage())) {name=Final Reflections, nodeColor=<rgba 0 0.2969 1 1>}
  n179 = EyeAdaptationInverse(LightValueInput=@n178, AlphaInput=Constant() {r=1})
  n180 = VectorParameter() {group=Mobile, parameterName=Mobile Color, defaultValue=<rgba 0.1507 0.2367 0.2865 1>}
  n181 = StaticSwitchParameter(True=LinearInterpolate(A=@n180.RGB, B=NamedRerouteUsage(), Alpha=NamedRerouteUsage()), False=@n180.RGB) {group=Texture Pattern, parameterName=Enable Texture Pattern, sortPriority=0}
  n182 = StaticSwitchParameter(
   True=Multiply(A=ScalarParameter() {defaultValue=0.5, group=Opaque, parameterName=Opaque Reflection Multiplier}, B=NamedRerouteUsage())
   False=ShadingPathSwitch(Default=Add(A=@n172, B=@n179.EyeAdaptationInverse), Mobile=Add(A=@n178, B=@n181))
  ) {group=Opaque, parameterName=IsOpaque}
  n183 = ScalarParameter() {defaultValue=5, group=VAT, parameterName=CrackBrightness}
  n184 = PreSkinnedPosition()
  n185 = Add(
   A=VertexInterpolator(VS=TransformPosition!(None=@n184) {transformType=TRANSFORMPOSSOURCE_World}).PS
   B=VectorParameter() {group=VAT, parameterName=Offset, defaultValue=<rgba 0 200 150 0>}.RGB
  )
  n186 = TextureSampleParameter2D!(
   UVs=Divide(A=ComponentMask(None=@n185) {b=true, g=true}, B=ScalarParameter() {defaultValue=325, parameterName=CrackSize})
   Apply View MipBias
  ) {
   group=VAT
   parameterName=Crackmage
   samplerSource=SSM_Wrap_WorldGroupSettings
   texture=/Game/Effects/Fort_Effects/Textures/Tile/T_Cracked_Ice_D
  }
  n187 = Multiply(A=Power(Base=@n186.R)) {constB=10}
  n188 = Add(A=TextureCoordinate(), B=Constant2Vector() {g=-0.1, r=0.1})
  n189 = Multiply(A=TextureSample(UVs=Multiply(A=@n188) {constB=1.4}) {texture=/Game/Effects/Fort_Effects/Textures/Tile/T_Glass_Cracks_M}.G) {constB=4}
  n190 = LinearInterpolate(
   A=Power(Base=Saturate(None=@n189))
   Alpha=MaterialFunctionCall(Radius=Constant() {r=0.5}, Density=Constant() {r=8}) {materialFunction=&RadialGradientExponential}.RadialGradientExponential
  ) {constB=0}
  StaticSwitchParameter(True=LinearInterpolate(A=@n182, B=Multiply(A=@n182, B=@n183) {constB=3}, Alpha=Max(A=@n187, B=@n190)), False=@n182) {parameterName=HasCrack}
  VectorParameter() {group=Opaque, parameterName=Opaque Color, defaultValue=<rgba 0.0623 0.0702 0.0781 1>}
 ]
}
