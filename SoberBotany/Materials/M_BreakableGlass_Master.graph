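# Unlit translucent master material for breakable glass, faking reflections and refracted background without scene lighting.
# Reflection vectors drive a spherically mapped horizon texture with manual derivatives, skylight cubemap and clouds, blended with Fresnel and sun specular.
# Periodic world-space triplanar projection supplies normals, masks and optional patterns, while time-of-day tints compute per vertex.
# Switches select tinted SceneColor, plain color or opaque output, with mobile and quality fallbacks.
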Material /SoberBotany/Materials/M_BreakableGlass_Master
graph {
 properties {
  bAutomaticallySetUsageInEditor=false
  bUsedWithInstancedStaticMeshes=true
  bUsedWithNanite=true
  blendMode=BLEND_Translucent
  physMaterial=/Game/FortressPhysicalMaterials/Glass
  pixelDepthOffsetMode=PDOM_Legacy
  shadingModel=MSM_Unlit
  usedShadingModels=Unlit
 }
 nodes [
  MaterialFunctionCall() {materialFunction=&MF_FN_DestructibleDamage}
  n1 = NamedRerouteUsage()
  n2 = Arctangent2Fast(Y=ComponentMask(None=@n1) {g=true}, X=ComponentMask(None=@n1) {r=true})
  n3 = MaterialFunctionCall(Multiplier=Constant() {r=2}) {materialFunction=&Pi}
  n4 = Multiply(
   A=Frac(None=Divide(A=@n2, B=@n3.Result))
   B=ScalarParameter() {defaultValue=4, group=Reflection, parameterName=Reflection Tiling X, sortPriority=1}
  ) {constB=4}
  n5 = Multiply(
   A=ArcsineFast!(None=ComponentMask(None=@n1) {b=true})
   B=Divide(B=MaterialFunctionCall() {materialFunction=&Pi}.Result) {constA=2}
  )
  n6 = Add(A=Multiply(A=Multiply(A=@n5) {constB=-1}, B=ScalarParameter() {defaultValue=2, group=Reflection, parameterName=Reflection Tiling Y, sortPriority=2}) {constB=2}) {constB=0.5}
  n7 = Add(A=Frac(None=@n4)) {constB=-0.5}
  n8 = ScalarParameter() {defaultValue=4, group=Reflection, parameterName=Reflection Tiling X, sortPriority=1}
  n9 = AppendVector(A=Multiply(A=Abs(None=@n7), B=Divide(B=@n8)), B=@n6)
  n10 = TextureSampleParameter2D!(UVs=AppendVector(A=@n4, B=@n6), "DDX(UVs)"=DDX(Value=@n9), "DDY(UVs)"=DDY(Value=@n9), Apply View MipBias) {mipValueMode=TMVM_Derivative, parameterName=Reflection Exterior, texture=&/T_BreakableGlass_Reflection_Hera_Exterior}
  n11 = ScalarParameter() {defaultValue=-3.5, group=Reflection, parameterName=Reflection Breakup Min}
  n12 = ScalarParameter() {defaultValue=1.2, group=Reflection, parameterName=Reflection Breakup Max}
  n13 = NamedRerouteUsage()
  Multiply(A=@n10.A, B=Saturate(None=LinearInterpolate(A=@n11, B=@n12, Alpha=@n13) {constA=3, constB=0.7}))
  n14 = VertexNormalWS()
  n15 = CameraVectorWS()
  n16 = OneMinus(None=Abs(None=DotProduct(A=@n14, B=@n15)))
  NamedRerouteDeclaration(None=Multiply(A=@n16, B=@n16)) {name=Fresnel, nodeColor=<rgba 1 0 0.0703 1>}
  n17 = Multiply(A=NamedRerouteUsage(), B=ScalarParameter() {defaultValue=0.01, group=Normal, parameterName=Normal Intensity, sortPriority=1}) {constB=0.1}
  n18 = VertexNormalWS()
  NamedRerouteDeclaration(None=ReflectionVectorWS(CustomWorldNormal=Add(A=@n17, B=@n18))) {name=Reflection Vector With Normals, nodeColor=<rgba 1 0.2578 0 1>}
  n19 = VectorParameter() {group=Specular, parameterName=Specular Color Morning, sortPriority=2, defaultValue=<rgba 2 0.9085 0.678 1>}
  n20 = VectorParameter() {group=Specular, parameterName=Specular Color Day, sortPriority=3, defaultValue=<rgba 1.4 1.2537 1.0938 1>}
  n21 = VectorParameter() {group=Specular, parameterName=Specular Color Evening, sortPriority=4, defaultValue=<rgba 2 0.7588 0.5208 1>}
  n22 = VectorParameter() {group=Specular, parameterName=Specular Color Night, sortPriority=5, defaultValue=<rgba 0.5391 0.7147 1.5 1>}
  NamedRerouteDeclaration(None=VertexInterpolator(VS=MaterialFunctionCall(Color_Morning=@n19.RGB, Color_Day=@n20.RGB, Color_Evening=@n21.RGB, Color_Night=@n22.RGB) {materialFunction=&Color_Time_of_Day}.Result).PS) {name=Specular Color, nodeColor=<rgba 0.5 0 1 1>}
  MaterialExpressionAbs_5 = Abs(None=VertexNormalWS())
  n23 = Convert(None=@MaterialExpressionAbs_5) {
   nodeName=Break Float3
   convertInputs [{
     type=Vector3
     expressionInput {
      expression=&self:MaterialExpressionAbs_5
      inputName=None
      mask=0
      maskA=0
      maskB=0
      maskG=0
      maskR=0
      outputIndex=0
     }
    }]
   convertOutputs [{type=Scalar}, {type=Scalar}, {type=Scalar}]
  }
  n24 = Add(A=@n23) {constB=0.2}
  n25 = Add(A=@n23) {constB=-0.2}
  n26 = AppendVector(A=Subtract(A=@n24, B=Max(A=@n23, B=@n25)), B=Subtract(A=@n23, B=Max(A=@n24, B=@n25)))
  NamedRerouteDeclaration(None=Saturate(None=Ceil(None=@n26))) {name=Triplanar Mask, nodeColor=<rgba 1 0 0 1>}
  n27 = VectorParameter() {group=Texture Pattern, parameterName=Texture Pattern Color 1, sortPriority=1, defaultValue=<rgba 0.5156 0.0645 0.1063 1>}
  n28 = VectorParameter() {group=Texture Pattern, parameterName=Texture Pattern Color 2, sortPriority=1, defaultValue=<rgba 0.5156 0.0645 0.1063 1>}
  n29 = WorldPosition() {worldPositionShaderOffset=WPT_CameraRelative}
  n30 = ScalarParameter() {defaultValue=13, group=Texture Pattern, parameterName=Texture Pattern Scale, sortPriority=0}
  n31 = VertexInterpolator(VS=Divide(
    A=TransformPosition!(None=@n29.XYZ, Periodic World Tile Size=@n30) {transformSourceType=TRANSFORMPOSSOURCE_TranslatedWorld, transformType=TRANSFORMPOSSOURCE_PeriodicWorld}
    B=@n30
   ))
  n32 = ComponentMask(None=@n31.PS) {g=true, r=true}
  n33 = ComponentMask(None=@n31.PS) {b=true, r=true}
  n34 = ComponentMask(None=NamedRerouteUsage()) {g=true}
  n35 = NamedRerouteUsage()
  n36 = TextureSampleParameter2D!(
   UVs=LinearInterpolate(A=LinearInterpolate(A=@n32, B=@n33, Alpha=@n34), B=ComponentMask(None=@n31.PS) {b=true, g=true}, Alpha=ComponentMask(None=@n35) {r=true})
   Apply View MipBias
  ) {
   group=Texture Pattern
   parameterName=Texture Pattern
   samplerType=SAMPLERTYPE_LinearGrayscale
   texture=&/T_BreakableGlass_Pattern_A
  }
  NamedRerouteDeclaration(None=LinearInterpolate(A=@n27.RGB, B=@n28.RGB, Alpha=@n36.R)) {name=Pattern Color, nodeColor=<rgba 1 0.3984 0 1>}
  NamedRerouteDeclaration(None=LinearInterpolate(
    A=ScalarParameter() {defaultValue=0.01, group=Texture Pattern, parameterName=Texture Pattern Opacity Low, sortPriority=2}
    B=ScalarParameter() {defaultValue=0.04, group=Texture Pattern, parameterName=Texture Pattern Opacity High, sortPriority=3}
    Alpha=@n36.R
   )) {name=Pattern Opacity, nodeColor=<rgba 0.0781 0 1 1>}
  n37 = WorldPosition() {worldPositionShaderOffset=WPT_CameraRelative}
  n38 = ScalarParameter() {defaultValue=300, group=Mask, parameterName=Mask Scale, sortPriority=1}
  n39 = VertexInterpolator(VS=Divide(
    A=TransformPosition(None=@n37.XYZ, Periodic World Tile Size=@n38) {transformSourceType=TRANSFORMPOSSOURCE_TranslatedWorld, transformType=TRANSFORMPOSSOURCE_PeriodicWorld}
    B=@n38
   ))
  n40 = NamedRerouteUsage()
  n41 = LinearInterpolate(
   A=LinearInterpolate(A=ComponentMask(None=@n39.PS) {g=true, r=true}, B=ComponentMask(None=@n39.PS) {b=true, r=true}, Alpha=ComponentMask(None=@n40) {g=true})
   B=ComponentMask(None=@n39.PS) {b=true, g=true}
   Alpha=ComponentMask(None=NamedRerouteUsage()) {r=true}
  )
  n42 = TextureSampleParameter2D(UVs=@n41) {group=Mask, parameterName=Mask, samplerType=SAMPLERTYPE_Masks, sortPriority=0, texture=&/T_GlassReflection_M}
  NamedRerouteDeclaration(None=LinearInterpolate(A=@n42.R, Alpha=NamedRerouteUsage()) {constB=0.97}) {name=Specular Mask, nodeColor=<rgba 1 0.5156 0 1>}
  n43 = TextureSampleParameter2D(UVs=@n41) {group=Normal, parameterName=Normal, samplerType=SAMPLERTYPE_LinearColor, sortPriority=0, texture=&/T_GlassReflection_N}
  n44 = Add(A=@n43.RGB) {constB=-0.5}
  n45 = NamedRerouteUsage()
  NamedRerouteDeclaration(None=ShadingPathSwitch(Default=LinearInterpolate(A=@n44, Alpha=@n45) {constB=0}, Mobile=Constant())) {name=Normals, nodeColor=<rgba 1 0.8438 0 1>}
  n46 = Subtract(A=@n25, B=Max(A=@n24, B=@n23))
  n47 = DotProduct(A=Saturate(None=AppendVector(A=@n26, B=@n46)), B=Constant() {r=1})
  NamedRerouteDeclaration(None=Power(Base=OneMinus(None=@n47)) {constExponent=32}) {name=Seam Mask, nodeColor=<rgba 0 0.5078 1 1>}
  n48 = VectorParameter() {group=Sky, parameterName=Sky Color Morning, sortPriority=4, defaultValue=<rgba 0.4062 0.3398 0.3281 1>}
  n49 = VectorParameter() {group=Sky, parameterName=Sky Color Day, sortPriority=5, defaultValue=<rgba 0.5469 0.5469 0.5469 1>}
  n50 = VectorParameter() {group=Sky, parameterName=Sky Color Evening, sortPriority=6, defaultValue=<rgba 0.4271 0.3328 0.3288 1>}
  n51 = VectorParameter() {group=Sky, parameterName=Sky Color Night, sortPriority=7, defaultValue=<rgba 0.4844 0.4844 0.4844 1>}
  NamedRerouteDeclaration(None=VertexInterpolator(VS=MaterialFunctionCall(Color_Morning=@n48.RGB, Color_Day=@n49.RGB, Color_Evening=@n50.RGB, Color_Night=@n51.RGB) {materialFunction=&Color_Time_of_Day}.Result).PS) {name=Sky Color, nodeColor=<rgba 0.2422 0 1 1>}
  n52 = VectorParameter() {group=Reflection, parameterName=Reflection Color Morning, sortPriority=3, defaultValue=<rgba 1.6 0.77 0.6474 1>}
  n53 = VectorParameter() {group=Reflection, parameterName=Reflection Color Day, sortPriority=4, defaultValue=<rgba 0.6146 0.6146 0.6146 1>}
  n54 = VectorParameter() {group=Reflection, parameterName=Reflection Color Evening, sortPriority=5, defaultValue=<rgba 1.4 0.7821 0.6612 1>}
  n55 = VectorParameter() {group=Reflection, parameterName=Reflection Color Night, sortPriority=6, defaultValue=<rgba 0.3061 0.4715 0.6875 1>}
  NamedRerouteDeclaration(None=VertexInterpolator(VS=MaterialFunctionCall(Color_Morning=@n52.RGB, Color_Day=@n53.RGB, Color_Evening=@n54.RGB, Color_Night=@n55.RGB) {materialFunction=&Color_Time_of_Day}.Result).PS) {name=Reflection Color, nodeColor=<rgba 1 0 0.3984 1>}
  n56 = LinearInterpolate(
   A=ScalarParameter() {defaultValue=4, group=Sky, parameterName=Sky Mip Bias High, sortPriority=3}
   B=ScalarParameter() {group=Sky, parameterName=Sky Mip Bias Low, sortPriority=2}
   Alpha=NamedRerouteUsage()
  )
  n57 = Desaturation(
   None=SkyLightEnvMapSample(Direction=NamedRerouteUsage(), Roughness=Saturate(None=@n56))
   Fraction=Multiply(A=ScalarParameter() {group=Sky, parameterName=Sky Saturation Compensation, sortPriority=0}) {constB=-1}
  )
  n58 = ComponentMask(None=VertexNormalWS()) {b=true}
  n59 = OneMinus(None=Multiply(A=ScalarParameter() {defaultValue=0.5, group=Sky, parameterName=Sky Top Reflection Suppress, sortPriority=1}, B=Abs(None=@n58)) {constB=0.7})
  NamedRerouteDeclaration(None=Multiply(A=Multiply(A=@n57, B=@n59), B=NamedRerouteUsage())) {name=Sky, nodeColor=<rgba 0 0.7188 1 1>}
  n60 = NamedRerouteUsage()
  n61 = ComponentMask(None=@n60) {g=true, r=true}
  n62 = ComponentMask(None=@n60) {b=true}
  n63 = LinearInterpolate(Alpha=Abs(None=NamedRerouteDeclaration(None=@n62) {name=Ref B, nodeColor=<rgba 0.0781 1 0 1>})) {constA=0.3}
  n64 = TextureSample(UVs=Multiply(A=Divide(A=@n61, B=@n63), B=ScalarParameter() {defaultValue=0.4, group=Clouds, parameterName=Clouds Tiling}) {constB=0.5}) {samplerType=SAMPLERTYPE_LinearColor, texture=/Game/TimeOfDay/TODM_Sequence/Materials/Sky/Textures/T_Ch6_OvercastClouds}
  NamedRerouteDeclaration(None=Multiply(A=@n64.R, B=NamedRerouteUsage())) {name=Clouds RGB, nodeColor=<rgba 1 0.9609 0 1>}
  n65 = Sign(None=@n62)
  n66 = SmoothStep(
   Min=ScalarParameter() {defaultValue=0.544, group=Clouds, parameterName=Clouds Range Min}
   Max=ScalarParameter() {defaultValue=0.8316, group=Clouds, parameterName=Clouds Range Max}
   Value=Multiply(A=@n64.B, B=Saturate(None=@n65))
  )
  NamedRerouteDeclaration(None=Multiply(A=Saturate(None=@n66), B=ScalarParameter() {defaultValue=0.7, group=Clouds, parameterName=Clouds Opacity})) {name=Clouds A, nodeColor=<rgba 0.6406 0 1 1>}
  NamedRerouteDeclaration(None=Multiply(A=@n10.RGB, B=NamedRerouteUsage())) {name=Horizon Reflection RGB, nodeColor=<rgba 0 1 0.3203 1>}
  NamedRerouteDeclaration(None=@n10.A) {name=Horizon Reflection A, nodeColor=<rgba 1 0 0 1>}
  n67 = SceneColor!(UVs)
  n68 = ScalarParameter() {defaultValue=0.6, group=SceneColor, parameterName=SceneColor Highlight Suppress}
  n69 = ScalarParameter() {group=SceneColor, parameterName=SceneColor Highlight Low}
  n70 = ScalarParameter() {defaultValue=0.2, group=SceneColor, parameterName=SceneColor Highlight High}
  n71 = Desaturation(None=@n67.RGB, Fraction=Constant() {r=1})
  n72 = Multiply(
   A=LinearInterpolate(A=@n67.RGB, B=Multiply(A=@n67.RGB, B=@n68) {constB=0.5}, Alpha=SmoothStep(Min=@n69, Max=@n70, Value=@n71) {constMax=0.7, constMin=0.2})
   B=VectorParameter() {group=SceneColor, parameterName=SceneColor Color, defaultValue=<rgba 0 1.9585 2 1>}.RGB
  )
  n73 = LinearInterpolate(A=@n72, B=Multiply(A=@n72, B=@n72), Alpha=NamedRerouteUsage())
  n74 = NamedRerouteUsage()
  n75 = Constant() {r=1}
  n76 = StaticSwitchParameter(
   True=LinearInterpolate(A=@n73, B=EyeAdaptationInverse(LightValueInput=@n74, AlphaInput=@n75).EyeAdaptationInverse, Alpha=NamedRerouteUsage())
   False=@n73
  ) {group=Texture Pattern, parameterName=Enable Texture Pattern, sortPriority=0}
  n77 = Multiply(A=@n76) {constB=0.7}
  NamedRerouteDeclaration(None=QualitySwitch(Default=@n76, Low=@n77, High=@n76, Medium=@n77, Epic=@n76)) {name=Scene Color, nodeColor=<rgba 1 0.4688 0 1>}
  n78 = SkyAtmosphereLightDirection()
  n79 = CameraVectorWS()
  n80 = NamedRerouteUsage()
  n81 = ScalarParameter() {defaultValue=0.0012, group=Specular, parameterName=Specular Normal Intensity, sortPriority=1}
  n82 = DotProduct(A=Normalize(VectorInput=Add(A=@n78, B=@n79)), B=Add(A=Multiply(A=@n80, B=@n81) {constB=0.1}, B=PixelNormalWS()))
  n83 = NamedRerouteUsage()
  n84 = Multiply(
   A=Saturate(None=LinearInterpolate(Alpha=@n83) {constA=20, constB=0.5})
   B=ScalarParameter() {defaultValue=300, group=Specular, parameterName=Specular Falloff, sortPriority=0}
  )
  NamedRerouteDeclaration(None=Multiply(A=Power(Base=@n82, Exp=@n84) {constExponent=300}, B=NamedRerouteUsage())) {name=Specular Highlight, nodeColor=<rgba 1 0 0.1641 1>}
  StaticSwitchParameter(
   True=ShadingPathSwitch(
    Default=ScalarParameter() {defaultValue=0.4, group=SceneColor, parameterName=SceneColor Opacity}
    Mobile=ScalarParameter() {defaultValue=0.6, group=Mobile, parameterName=Mobile Opacity}
   )
   False=ScalarParameter() {defaultValue=0.3, group=Color, parameterName=Opacity}
  ) {defaultValue=true, parameterName=UseSceneColor}
  n85 = NamedRerouteUsage()
  n86 = NamedRerouteUsage()
  n87 = NamedRerouteUsage()
  n88 = Multiply(
   A=LinearInterpolate(A=LinearInterpolate(A=@n85, B=@n86, Alpha=@n87), B=NamedRerouteUsage(), Alpha=NamedRerouteUsage())
   B=LinearInterpolate(
    A=ScalarParameter() {defaultValue=0.2, group=Fresnel, parameterName=Fresnel Reflection Intensity}
    B=ScalarParameter() {defaultValue=1.4, group=Fresnel, parameterName=Fresnel Reflection Intensity 90º}
    Alpha=NamedRerouteUsage()
   )
  ) {constB=0.2}
  n89 = Add(A=@n88, B=NamedRerouteUsage())
  n90 = NamedRerouteDeclaration(None=Multiply(A=ShadingPathSwitch(Default=@n89, Mobile=@n88), B=NamedRerouteUsage())) {name=Final Reflections, nodeColor=<rgba 0 0.2969 1 1>}
  n91 = EyeAdaptationInverse(LightValueInput=Add(A=NamedRerouteUsage(), B=@n90), AlphaInput=Constant() {r=0.7})
  n92 = VectorParameter() {group=Color, parameterName=Mobile Color, defaultValue=<rgba 0.1507 0.2367 0.2865 1>}
  n93 = NamedRerouteUsage()
  n94 = NamedRerouteUsage()
  n95 = Add(
   A=@n90
   B=StaticSwitchParameter(True=LinearInterpolate(A=@n92.RGB, B=@n93, Alpha=@n94), False=@n92.RGB) {group=Texture Pattern, parameterName=Enable Texture Pattern, sortPriority=0}
  )
  n96 = VectorParameter() {group=Color, parameterName=Glass Color Morning, sortPriority=4, defaultValue=<rgba -0.01 0.02 0.03 1>}
  n97 = VectorParameter() {group=Color, parameterName=Glass Color Evening, sortPriority=6, defaultValue=<rgba -0.01 0.02 0.03 1>}
  n98 = VectorParameter() {group=Color, parameterName=Glass Color Night, sortPriority=7, defaultValue=<rgba -0.01 0.02 0.03 1>}
  n99 = VertexInterpolator(VS=MaterialFunctionCall(
    Color_Morning=@n96.RGB
    Color_Day=VectorParameter() {group=Color, parameterName=Glass Color Day, sortPriority=5, defaultValue=<rgba -0.01 0.02 0.03 1>}.RGB
    Color_Evening=@n97.RGB
    Color_Night=@n98.RGB
   ) {materialFunction=&Color_Time_of_Day}.Result)
  n100 = NamedRerouteUsage()
  n101 = NamedRerouteUsage()
  n102 = Add(
   A=Multiply(A=@n90, B=ScalarParameter() {defaultValue=1, group=Color, parameterName=Reflection Boost})
   B=StaticSwitchParameter(True=LinearInterpolate(A=@n99.PS, B=@n100, Alpha=@n101), False=@n99.PS) {group=Texture Pattern, parameterName=Enable Texture Pattern, sortPriority=0}
  )
  n103 = Constant() {r=0.5}
  StaticSwitchParameter(
   True=Multiply(A=ScalarParameter() {defaultValue=0.5, group=Opaque, parameterName=Opaque Reflection Multiplier}, B=NamedRerouteUsage())
   False=StaticSwitchParameter(
    True=ShadingPathSwitch(Default=@n91.EyeAdaptationInverse, Mobile=@n95)
    False=EyeAdaptationInverse(LightValueInput=@n102, AlphaInput=@n103).EyeAdaptationInverse
   ) {defaultValue=true, parameterName=UseSceneColor}
  ) {group=Opaque, parameterName=IsOpaque}
  VectorParameter() {group=Opaque, parameterName=Opaque Color, defaultValue=<rgba 0.0623 0.0702 0.0781 1>}
 ]
}
