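# M_ForceMipStreaming is an unlit material, apparently a helper for forcing texture mip streaming, judging by its name.
# A UV Tiling parameter scales texture coordinates around the UV center.
# Those UVs sample a swappable Texture parameter, defaulting to the exterior breakable-glass reflection, and a fixed interior reflection texture.
# A lerp with unwired alpha blends both RGB results, and no connection from it to a material output is shown.
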Material /SoberBotany/Materials/M_ForceMipStreaming
graph {
 properties {bAutomaticallySetUsageInEditor=false, shadingModel=MSM_Unlit, usedShadingModels=Unlit}
 nodes [
  n1 = TextureCoordinate()
  n2 = Add(A=Multiply(A=Add(A=@n1) {constB=-0.5}, B=ScalarParameter() {defaultValue=0.2, parameterName=UV Tiling})) {constB=0.5}
  LinearInterpolate(
   A=TextureSampleParameter2D(UVs=@n2) {parameterName=Texture, texture=&/T_BreakableGlass_Reflection_Hera_Exterior}.RGB
   B=TextureSample(UVs=@n2) {texture=&/T_BreakableGlass_Reflection_Hera_Interior}.RGB
  )
 ]
}
